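# Emissive material for the Storm Tracker backpack's orange wrist screen: an animated, pixelated readout on a dark Default Lit surface.
# Custom HLSL sums time-driven sines on floor-quantized UV1 coordinates, drawing a glowing waveform and a bar-like wave that modulates the frame mask within a box.
# A derivative-antialiased cell grid cuts both waves and the time texture.
# Tinted waves, frame and box-mask glow add to the time texture, then pass through accessory tech-art overrides.
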Material /BRCosmetics/Accessories/FORT_Backpacks/M_MED_StormTracker_Backpack/Materials/M_StormTrackerOrangeScreenWrist
graph {
 properties {
  bUseMaterialAttributes=true
  bUsedWithMeshDeformer=true
  bUsedWithSkeletalMesh=true
  pixelDepthOffsetMode=PDOM_Legacy
  usedShadingModels=Default Lit
 }
 nodes [
  n1 = MaterialFunctionCall(
   X=ScalarParameter() {defaultValue=8, group=EQ, parameterName=quantFactorX2}
   Y=ScalarParameter() {defaultValue=100, group=EQ, parameterName=quantFactorY2}
  ) {materialFunction=&MakeFloat2}
  n2 = VectorParameter() {parameterName=SineWaveScale2, defaultValue=<rgba 3 3 0 1>}
  n3 = TextureCoordinate() {coordinateIndex=1}
  n4 = MaterialFunctionCall(UVs=@n3, Texture Scale=VectorParameter() {parameterName=Param1, defaultValue=<rgba 0.74 0.65 0 0>}.RGB) {materialFunction=&ScaleUVsByCenter}
  n5 = Multiply(
   A=@n4.UVs
   B=MaterialFunctionCall(X=@n2.R, Y=@n2.R) {materialFunction=&MakeFloat2}.Result
  ) {constB=3}
  n6 = VectorParameter() {parameterName=SineWaveOffset2, defaultValue=<rgba -2 -2.1 0 1>}
  n7 = MaterialFunctionCall(X=@n6.R, Y=@n6.R) {materialFunction=&MakeFloat2}
  n8 = Multiply(
   A=MaterialFunctionCall(XY=Add(A=@n5, B=@n7.Result) {constB=0}) {materialFunction=&Swizzle}.YX
   B=@n1.Result
  ) {constB=5}
  MaterialExpressionReroute_31 = Reroute(None=Divide(A=Floor(None=@n8), B=@n1.Result) {constB=5})
  MaterialExpressionConstant_6 = Constant() {r=0.01}
  MaterialExpressionScalarParameter_32 = ScalarParameter() {defaultValue=2.1, group=SineWave, parameterName=Frequency2}
  MaterialExpressionScalarParameter_33 = ScalarParameter() {defaultValue=10, group=SineWave, parameterName=Amplitude2}
  MaterialExpressionScalarParameter_34 = ScalarParameter() {defaultValue=1, group=SineWave, parameterName=Speed2}
  MaterialExpressionTime_0 = Time()
  MaterialExpressionCustom_1 = Custom!(
   uv=@MaterialExpressionReroute_31
   frequency=@MaterialExpressionScalarParameter_32
   amplitude=@MaterialExpressionScalarParameter_33
   speed=@MaterialExpressionScalarParameter_34
   itime=@MaterialExpressionTime_0
  ) {
   code="  \tfloat phase = itime * speed;\r\n \t//amplitude at position x (y value)\r\n  \tfloat yOffset = sin(uv.x * frequency + phase);\r\n    yOffset += sin(uv.x*frequency*2.1 + phase)*4.5;\r\n    //yOffset += sin(uv.x*frequency*1.72 + phase*1.121)*4.0;\r\n\t//yOffset += sin(uv.x*frequency*2.221 + phase*0.437)*5.0;\r\n\t//yOffset += sin(uv.x*frequency*3.1122+ phase*4.269)*2.5;  \r\n  \treturn yOffset * amplitude;"
   description=CreateWave
   outputType=CMOT_Float1
   inputs [
    {
     inputName=uv
     input {expression=&self:MaterialExpressionReroute_31}
    }
    {
     inputName=frequency
     input {expression=&self:MaterialExpressionScalarParameter_32}
    }
    {
     inputName=amplitude
     input {expression=&self:MaterialExpressionScalarParameter_33}
    }
    {
     inputName=speed
     input {expression=&self:MaterialExpressionScalarParameter_34}
    }
    {
     inputName=itime
     input {expression=&self:MaterialExpressionTime_0}
    }
   ]
  }
  n9 = Custom!(uv=@MaterialExpressionReroute_31, lineWidth=@MaterialExpressionConstant_6, wave=@MaterialExpressionCustom_1) {
   code="float result = ((uv.y-0.5) / lineWidth) + wave;    \r\nreturn (result);"
   description=DrawWave
   outputType=CMOT_Float1
   inputs [
    {
     inputName=uv
     input {expression=&self:MaterialExpressionReroute_31}
    }
    {
     inputName=lineWidth
     input {expression=&self:MaterialExpressionConstant_6}
    }
    {
     inputName=wave
     input {expression=&self:MaterialExpressionCustom_1}
    }
   ]
  }
  n10 = Saturate(None=@n9)
  Add(A=@n10) {constB=0.25}
  n11 = MaterialFunctionCall(
   X=ScalarParameter() {defaultValue=25, group=EQ, parameterName=BGquantFactorX}
   Y=ScalarParameter() {defaultValue=25, group=EQ, parameterName=BGquantFactorY}
  ) {materialFunction=&MakeFloat2}
  n12 = Multiply(A=@n4.UVs, B=@n11.Result) {constB=5}
  Floor(None=@n12)
  n13 = VectorParameter() {parameterName=FrameScale, defaultValue=<rgba 0.7 1 0 0>}
  n14 = MaterialFunctionCall(X=@n13.R, Y=@n13.R) {materialFunction=&MakeFloat2}
  n15 = MaterialFunctionCall(
   X=ScalarParameter() {defaultValue=75, group=EQ, parameterName=TextquantFactorX}
   Y=ScalarParameter() {defaultValue=50, group=EQ, parameterName=TextquantFactorY}
  ) {materialFunction=&MakeFloat2}
  n16 = Divide(A=Floor(None=Multiply(A=@n4.UVs, B=@n15.Result) {constB=5}), B=@n15.Result) {constB=5}
  n17 = VectorParameter() {parameterName=FrameOffset, defaultValue=<rgba 0.145 0 0 0>}
  n18 = TextureObject() {
   samplerType=SAMPLERTYPE_LinearGrayscale
   texture=/BRCosmetics/Accessories/FORT_Backpacks/M_MED_StormTracker_Backpack/Textures/T_StormTrackerOrangeScreenTime_M
  }
  n19 = TextureSample(
   UVs=Add(
    A=Multiply(A=@n16, B=@n14.Result) {constB=3}
    B=MaterialFunctionCall(X=@n17.R, Y=@n17.R) {materialFunction=&MakeFloat2}.Result
   ) {constB=0}
   Tex=@n18
  ) {
   samplerSource=SSM_Clamp_WorldGroupSettings
   samplerType=SAMPLERTYPE_LinearGrayscale
   texture=&T_StormTrackerOrangeScreenTime_M
  }
  MaterialExpressionScalarParameter_19 = ScalarParameter() {defaultValue=0.1, group=EQ, parameterName=LineWidth, sliderMax=1}
  n20 = VectorParameter() {parameterName=SineWaveScale, defaultValue=<rgba 3 3 0 1>}
  n21 = VectorParameter() {parameterName=SineWaveOffset, defaultValue=<rgba -0.3 -1.3 0 1>}
  n22 = Add(
   A=Multiply(
    A=@n4.UVs
    B=MaterialFunctionCall(X=@n20.R, Y=@n20.R) {materialFunction=&MakeFloat2}.Result
   ) {constB=3}
   B=MaterialFunctionCall(X=@n21.R, Y=@n21.R) {materialFunction=&MakeFloat2}.Result
  ) {constB=0}
  n23 = MaterialFunctionCall(
   X=ScalarParameter() {defaultValue=25, group=EQ, parameterName=quantFactorX}
   Y=ScalarParameter() {defaultValue=25, group=EQ, parameterName=quantFactorY}
  ) {materialFunction=&MakeFloat2}
  n24 = Frac(None=Multiply(A=@n22, B=@n23.Result))
  n25 = Min(A=@n24, B=OneMinus(None=@n24))
  n26 = Add(A=Abs(None=DDX(Value=@n25)), B=Abs(None=DDY(Value=@n25)))
  MaterialExpressionSubtract_9 = Subtract(A=@n25, B=@n26)
  MaterialExpressionAdd_19 = Add(A=@n25, B=@n26)
  n27 = Custom!(varA=@MaterialExpressionScalarParameter_19, varMin=@MaterialExpressionSubtract_9, varMax=@MaterialExpressionAdd_19) {
   code="return smoothstep(varMin,varMax,varA);"
   description=smoothStep_V2
   outputType=CMOT_Float2
   inputs [
    {
     inputName=varA
     input {expression=&self:MaterialExpressionScalarParameter_19}
    }
    {
     inputName=varMin
     input {expression=&self:MaterialExpressionSubtract_9}
    }
    {
     inputName=varMax
     input {expression=&self:MaterialExpressionAdd_19}
    }
   ]
  }
  n28 = MaterialFunctionCall(Float2=@n27) {materialFunction=&BreakOutFloat2Components}
  n29 = OneMinus(None=Saturate(None=Add(A=@n28.R, B=@n28.R)))
  n30 = Multiply(A=Multiply(A=@n19.RGB), B=@n29)
  n31 = VectorParameter() {parameterName=FrameScale, defaultValue=<rgba 0.7 1 0 0>}
  n32 = MaterialFunctionCall(X=@n31.R, Y=@n31.R) {materialFunction=&MakeFloat2}
  n33 = VectorParameter() {parameterName=FrameOffset, defaultValue=<rgba 0.145 0 0 0>}
  n34 = TextureObject() {
   samplerType=SAMPLERTYPE_Masks
   texture=/BRCosmetics/Accessories/FORT_Backpacks/M_MED_StormTracker_Backpack/Textures/T_StormTrackerOrangeScreenFrame_M
  }
  n35 = TextureSample(
   UVs=Add(
    A=Multiply(A=@n4.UVs, B=@n32.Result)
    B=MaterialFunctionCall(X=@n33.R, Y=@n33.R) {materialFunction=&MakeFloat2}.Result
   ) {constB=0}
   Tex=@n34
  ) {
   samplerSource=SSM_Clamp_WorldGroupSettings
   samplerType=SAMPLERTYPE_Masks
   texture=&T_StormTrackerOrangeScreenFrame_M
  }
  n36 = Multiply(A=@n35.RGB) {constB=0.3}
  n37 = Saturate(None=Max(A=Min(A=@n10) {constB=0.7}) {constB=0.2})
  n38 = VectorParameter() {parameterName=RedSquareScale, defaultValue=<rgba 4.7 9 0 1>}
  n39 = MaterialFunctionCall(X=@n38.R, Y=@n38.R) {materialFunction=&MakeFloat2}
  n40 = VectorParameter() {parameterName=RedSquareOffset, defaultValue=<rgba -3.8583 -2.8 0 1>}
  n41 = MaterialFunctionCall!(
   A=Add(
    A=Multiply(A=@n4.UVs, B=@n39.Result) {constB=3}
    B=MaterialFunctionCall(X=@n40.R, Y=@n40.R) {materialFunction=&MakeFloat2}.Result
   ) {constB=0}
   B=Constant3Vector() {constant=<rgba 0.5 0.5 0 0>}
   Bounds=ScalarParameter() {defaultValue=1.8, parameterName=RedSquare-BoxBounds}
   Edge Falloff=ScalarParameter() {defaultValue=0.05, parameterName=RedSquare-BoxFalloff}
  ) {materialFunction=/Engine/Functions/Engine_MaterialFunctions01/Gradient/BoxMask-2D}
  n42 = Saturate(None=Add(A=Multiply(A=@n29, B=@n37), B=OneMinus(None=@n41.Result)))
  n43 = MaterialFunctionCall(
   A=@n3
   B=Divide(A=Round(None=@n12), B=@n11.Result) {constB=5}
   Bounds=ScalarParameter() {parameterName=MainScale-BoxBounds}
   Edge Falloff=ScalarParameter() {defaultValue=1.5, parameterName=MainScale-BoxFalloff}
  ) {materialFunction=&BoxMask-2D}
  n44 = Power(
   Base=Multiply(A=OneMinus(None=@n43.Result), B=ScalarParameter() {defaultValue=3, parameterName=int}) {constB=2}
   Exp=ScalarParameter() {defaultValue=2.5, parameterName=pow}
  ) {constExponent=1}
  n45 = Multiply(A=@n22, B=@n23.Result) {constB=5}
  MaterialExpressionReroute_9 = Reroute(None=Divide(A=Floor(None=@n45), B=@n23.Result) {constB=5})
  MaterialExpressionScalarParameter_45 = ScalarParameter() {defaultValue=0.1, parameterName=LineWidth}
  MaterialExpressionScalarParameter_12 = ScalarParameter() {defaultValue=12, group=SineWave, parameterName=Frequency}
  MaterialExpressionScalarParameter_13 = ScalarParameter() {defaultValue=0.7, group=SineWave, parameterName=Amplitude}
  MaterialExpressionScalarParameter_14 = ScalarParameter() {defaultValue=18, group=SineWave, parameterName=Speed}
  MaterialExpressionTime_3 = Time()
  n46 = Custom!(
   uv=@MaterialExpressionReroute_9
   frequency=@MaterialExpressionScalarParameter_12
   amplitude=@MaterialExpressionScalarParameter_13
   speed=@MaterialExpressionScalarParameter_14
   itime=@MaterialExpressionTime_3
  ) {
   code="  \tfloat phase = itime * speed;\r\n \t//amplitude at position x (y value)\r\n  \tfloat yOffset = sin(uv.x * frequency + phase);\r\n    yOffset += sin(uv.x*frequency*2.1 + phase)*4.5;\r\n    yOffset += sin(uv.x*frequency*1.72 + phase*1.121)*4.0;\r\n\t//yOffset += sin(uv.x*frequency*2.221 + phase*0.437)*5.0;\r\n\tyOffset += sin(uv.x*frequency*3.1122+ phase*4.269)*2.5;  \r\n  \treturn yOffset * amplitude;"
   description=CreateWave
   outputType=CMOT_Float1
   inputs [
    {
     inputName=uv
     input {expression=&self:MaterialExpressionReroute_9}
    }
    {
     inputName=frequency
     input {expression=&self:MaterialExpressionScalarParameter_12}
    }
    {
     inputName=amplitude
     input {expression=&self:MaterialExpressionScalarParameter_13}
    }
    {
     inputName=speed
     input {expression=&self:MaterialExpressionScalarParameter_14}
    }
    {
     inputName=itime
     input {expression=&self:MaterialExpressionTime_3}
    }
   ]
  }
  n47 = Power(
   Base=Distance(
    A=ScalarParameter() {defaultValue=0.5, parameterName=WaveCenter}
    B=MaterialFunctionCall(Float2=@MaterialExpressionReroute_9) {materialFunction=&BreakOutFloat2Components}.R
   )
   Exp=ScalarParameter() {defaultValue=0.7, parameterName=WaveWidth}
  ) {constExponent=1.5}
  MaterialExpressionMultiply_3 = Multiply(A=@n46, B=OneMinus(None=Saturate(None=@n47)))
  n48 = Custom!(uv=@MaterialExpressionReroute_9, lineWidth=@MaterialExpressionScalarParameter_45, wave=@MaterialExpressionMultiply_3) {
   code="float result = ((uv.y-0.5) / lineWidth) + wave;    \r\nreturn (result);"
   description=DrawWave
   outputType=CMOT_Float1
   inputs [
    {
     inputName=uv
     input {expression=&self:MaterialExpressionReroute_9}
    }
    {
     inputName=lineWidth
     input {expression=&self:MaterialExpressionScalarParameter_45}
    }
    {
     inputName=wave
     input {expression=&self:MaterialExpressionMultiply_3}
    }
   ]
  }
  n49 = Multiply(A=Saturate(None=@n48), B=ScalarParameter() {defaultValue=0.1, parameterName=SineWaveLowEmissive}) {constB=0.3}
  n50 = Abs(None=@n48)
  n51 = Multiply(A=Saturate(None=OneMinus(None=@n50)), B=ScalarParameter() {defaultValue=50, parameterName=SineWaveEmissive}) {constB=10}
  n52 = Multiply(
   A=MaterialFunctionCall(
    A=@n22
    B=Constant3Vector() {constant=<rgba 0.5 0.2 0 0>}
    Bounds=ScalarParameter() {defaultValue=2, parameterName=SineWave-BoxBounds}
    Edge Falloff=ScalarParameter() {defaultValue=0.05, parameterName=SineWave-BoxFalloff}
   ) {materialFunction=&BoxMask-2D}.Result
   B=Multiply(A=@n29, B=Add(A=@n49, B=@n51))
  )
  n53 = Multiply(
   A=VectorParameter() {parameterName=EmissiveColor, defaultValue=<rgba 1 0.05 0 0>}.RGB
   B=Add(A=Multiply(A=@n36, B=@n42), B=Add(A=@n44, B=@n52))
  )
  MaterialFunctionCall(In=MakeMaterialAttributes!(
    BaseColor=Constant() {r=0.01}
    Metallic
    Specular=ScalarParameter() {defaultValue=0.1, parameterName=Spec}
    Roughness=ScalarParameter() {defaultValue=0.7, parameterName=Rough}
    Anisotropy
    EmissiveColor=Add(A=@n30, B=@n53)
    Opacity
    OpacityMask
    Normal
    Tangent
    WorldPositionOffset
    SubsurfaceColor
    ClearCoat
    ClearCoatRoughness
    AmbientOcclusion
    Refraction
    CustomizedUVs_0
    CustomizedUVs_1
    CustomizedUVs_2
    CustomizedUVs_3
    CustomizedUVs_4
    CustomizedUVs_5
    CustomizedUVs_6
    CustomizedUVs_7
    PixelDepthOffset
    ShadingModel
    Displacement
   )) {materialFunction=&MF_TechArtOverrides_accessories}
 ]
}
