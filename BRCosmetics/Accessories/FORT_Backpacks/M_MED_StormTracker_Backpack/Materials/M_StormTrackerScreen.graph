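# Screen material for the Storm Tracker backpack, showing an animated radar-style minimap of the storm as emissive on a dark lit surface.
# A rotating angular sweep lights a dotted map texture and noise-distorted concentric storm lines.
# CurrentCircle and NextCircle from the minimap parameter collection drive the safe-zone blend and next-circle ring, using UVs rotated to correct the mesh.
# A box mask fades screen edges, while Low quality uses simpler effects without that mask.
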Material /BRCosmetics/Accessories/FORT_Backpacks/M_MED_StormTracker_Backpack/Materials/M_StormTrackerScreen
graph {
 properties {
  bUseMaterialAttributes=true
  bUsedWithMeshDeformer=true
  bUsedWithSkeletalMesh=true
  pixelDepthOffsetMode=PDOM_Legacy
  usedShadingModels=Default Lit
 }
 nodes [
  Constant3Vector() {desc="Debug \"current\" values", constant=<rgba 0.4 0.55 0.52 1>}
  Constant3Vector() {desc="Debug \"next\" values", constant=<rgba 0.3429 0.4452 0.25 1>}
  Constant() {r=0.5}
  Constant() {r=0.125}
  Constant() {r=0.085}
  n1 = MaterialFunctionCall!(
   A=TextureCoordinate() {coordinateIndex=1}
   B=Constant2Vector() {g=0.5, r=0.5}
   Bounds=MaterialFunctionCall(X=ScalarParameter() {defaultValue=0.78, parameterName=Screenedge-X}, Y=ScalarParameter() {defaultValue=0.58, parameterName=Screenedge-Y}) {materialFunction=&MakeFloat2}.Result
   Edge Falloff=ScalarParameter() {defaultValue=0.12, parameterName=Screenedge-Falloff}
  ) {materialFunction=/Engine/Functions/Engine_MaterialFunctions01/Gradient/BoxMask-2D}
  n2 = Time()
  n3 = ScalarParameter() {defaultValue=1.5, parameterName=SweepSpeed}
  n4 = MaterialFunctionCall(Float2=Rotator(Coordinate=TextureCoordinate() {coordinateIndex=1}, Time=Multiply(A=@n2, B=@n3)) {speed=1}) {materialFunction=&BreakOutFloat2Components}
  n5 = OneMinus(None=@n4.G)
  n6 = OneMinus(None=@n4.R)
  n7 = MaterialFunctionCall(Vector or UVs=ConstantBiasScale(None=AppendVector(A=@n5, B=@n6)) {bias=-0.5, scale=2}) {materialFunction=&VectorToRadialValue}
  n8 = Multiply(A=Power(Base=@n7.Vector Converted to Angle) {constExponent=50}) {constB=0.5}
  n9 = VectorParameter() {group=EffectParams, parameterName=BackgroundMapColor, defaultValue=<rgba 0.0375 0.2982 0.75 1>}
  n10 = ScalarParameter() {defaultValue=0.125, parameterName=Dots-Tile}
  n11 = MaterialFunctionCall(UVs=TextureCoordinate() {coordinateIndex=1}, "Rotation Angle (0-1)"=ScalarParameter() {defaultValue=0.5, parameterName=MainRotAng}) {
   desc="UVs are rotated upside down in\r\nbackpack mesh"
   materialFunction=&CustomRotator
  }
  n12 = MaterialFunctionCall(UVs=@n11.Rotated Values, Texture Scale=ScalarParameter() {defaultValue=0.7, parameterName=MainScale}) {materialFunction=&ScaleUVsByCenter}
  n13 = Frac(None=Add(A=Constant2Vector() {g=0.5, r=0.5}, B=Divide(A=@n12.UVs, B=@n10) {constB=5}))
  n14 = Constant() {r=0.5}
  n15 = ScalarParameter() {defaultValue=0.3, parameterName=Dots-Radius}
  n16 = ScalarParameter() {parameterName=Dots-Hard}
  n17 = TextureSample(UVs=@n12.UVs) {samplerType=SAMPLERTYPE_Masks, texture=/Game/Accessories/FORT_Backpacks/M_MED_StormTracker_Backpack/Textures/T_StormTrackerScreen}
  n18 = Add(
   A=Multiply(A=SphereMask(A=@n13, B=@n14, Radius=@n15, Hardness=@n16), B=ScalarParameter() {defaultValue=0.3, parameterName=Dots-Amount})
   B=@n17.G
  )
  n19 = Power(Base=OneMinus(None=@n7.Vector Converted to Angle), Exp=ScalarParameter() {defaultValue=2, parameterName=GridSweepPower})
  n20 = Add(A=Add(A=@n18, B=@n19), B=Constant() {r=0.5})
  n21 = Multiply(
   A=MaterialFunctionCall(
    A=@n11.Rotated Values
    B=@n12.UVs
    Bounds=ScalarParameter() {defaultValue=0.3, parameterName=MainScale-BoxBounds}
    Edge Falloff=ScalarParameter() {defaultValue=0.001, parameterName=MainScale-BoxFalloff}
   ) {materialFunction=&BoxMask-2D}.Result
   B=@n17.R
  )
  n22 = Add(A=Multiply(A=@n18, B=@n19), B=Multiply(A=Saturate(None=@n20), B=@n21))
  n23 = VectorParameter() {group=EffectParams, parameterName=TintOutsideSafeZone, defaultValue=<rgba 0.3448 0.015 0.75 0.1>}
  n24 = Panner(Coordinate=Multiply(A=TextureCoordinate() {coordinateIndex=1}, B=ScalarParameter() {defaultValue=1.5, parameterName=Lines-NoiseSize})) {speedY=0.08}
  n25 = CollectionParameter() {collection=/Game/Athena/HUD/MiniMap/MPC_Minimap, parameterName=CurrentCircle}
  n26 = Distance(A=@n12.UVs, B=ComponentMask(None=@n25) {g=true, r=true})
  n27 = Add(
   A=Multiply(
    A=TextureSample(UVs=@n24) {samplerType=SAMPLERTYPE_LinearGrayscale, texture=/Game/Global/Textures/Default/Tileable_Patterns/T_StormCloudDetail01_M}.R
    B=ScalarParameter() {defaultValue=1.5, parameterName=Lines-NoiseAmount}
   )
   B=Multiply(A=ScalarParameter() {defaultValue=9, parameterName=Lines-Tile}, B=@n26)
  )
  n28 = ScalarParameter() {defaultValue=50, parameterName=Lines-PowA}
  n29 = ScalarParameter() {defaultValue=1, parameterName=Lines-PowB}
  n30 = Saturate(None=Multiply(A=ScalarParameter() {defaultValue=1.1, parameterName=Lines-TrailPush}, B=@n7.Vector Converted to Angle))
  n31 = Add(
   A=Power(
    Base=TextureSample(UVs=@n27) {samplerType=SAMPLERTYPE_LinearColor, texture=/Game/Global/Textures/Default/Masks_and_Gradients/T_VertGradient}.R
    Exp=LinearInterpolate(A=@n28, B=@n29, Alpha=@n30)
   )
   B=ScalarParameter() {defaultValue=0.25, parameterName=Lines-Add}
  )
  n32 = Power(Base=@n7.Vector Converted to Angle, Exp=ScalarParameter() {defaultValue=3, parameterName=CloudsSweepPower})
  n33 = ScalarParameter() {defaultValue=6, parameterName=CloudsSweepBright}
  n34 = ComponentMask(None=@n25) {b=true}
  n35 = Add(A=@n34, B=Multiply(A=Constant() {r=0.001}, B=ScalarParameter() {defaultValue=150, parameterName=Current-ScaleFactorOuter}))
  n36 = Multiply(
   A=OneMinus(None=MaterialFunctionCall(Alpha=@n26, Min=@n34, Max=@n35) {materialFunction=&SmoothStep}.Result)
   B=ScalarParameter() {defaultValue=0.25, parameterName=StormEdgeFalloffAmount}
  )
  n37 = Add(A=Multiply(A=Saturate(None=@n31), B=Multiply(A=@n32, B=@n33)), B=@n36)
  n38 = Constant() {r=0.001}
  n39 = ScalarParameter() {defaultValue=40, parameterName=Current-ScaleFactor}
  n40 = MaterialFunctionCall(Alpha=@n26, Min=Subtract(A=@n34, B=Multiply(A=@n38, B=@n39)), Max=@n34) {materialFunction=&SmoothStep}
  n41 = Multiply(A=ScalarParameter() {defaultValue=0.75, parameterName=IslandInsideStormEffect}, B=@n21) {constB=0.25}
  n42 = Add(A=@n23.A, B=Add(A=Multiply(A=@n37, B=@n40.Result), B=@n41))
  n43 = Add(A=@n8, B=LinearInterpolate(A=Multiply(A=@n9.RGB, B=@n22), B=Multiply(A=@n23.RGB, B=@n42), Alpha=@n40.Result))
  n44 = CollectionParameter() {collection=&MPC_Minimap, parameterName=NextCircle}
  n45 = Distance(A=@n12.UVs, B=ComponentMask(None=@n44) {g=true, r=true})
  n46 = Multiply(A=Constant() {r=0.001}, B=ScalarParameter() {defaultValue=20, parameterName=Next-ScaleFactor})
  n47 = ComponentMask(None=@n44) {b=true}
  n48 = MaterialFunctionCall(Alpha=@n45, Min=Subtract(A=@n47, B=@n46), Max=@n47) {materialFunction=&SmoothStep}
  n49 = MaterialFunctionCall(Alpha=@n45, Min=@n47, Max=Add(A=@n47, B=@n46)) {materialFunction=&SmoothStep}
  n50 = Multiply(
   A=VectorParameter() {group=EffectParams, parameterName=NextCircleTint, defaultValue=<rgba 20 1.4991 0 1>}.RGB
   B=Subtract(A=Add(A=@n48.Result), B=Add(A=@n49.Result))
  )
  n51 = TextureSampleParameter2D() {
   parameterName=Normal
   samplerType=SAMPLERTYPE_Normal
   texture=/Game/Accessories/FORT_Backpacks/M_MED_StormTracker_Backpack/Textures/T_M_MED_StormTracker_Backpack_N
  }
  n52 = MakeMaterialAttributes!(
   BaseColor=Constant() {r=0.01}
   Metallic
   Specular=Multiply(A=ScalarParameter() {defaultValue=0.1, parameterName=Spec}, B=@n1.Result)
   Roughness=ScalarParameter() {defaultValue=0.6, parameterName=Rough}
   Anisotropy
   EmissiveColor=Multiply(A=Add(A=@n43, B=@n50), B=@n1.Result) {constA=1}
   Opacity
   OpacityMask
   Normal=@n51.RGB
   Tangent
   WorldPositionOffset
   SubsurfaceColor
   ClearCoat
   ClearCoatRoughness
   AmbientOcclusion
   Refraction
   CustomizedUVs_0
   CustomizedUVs_1
   CustomizedUVs_2
   CustomizedUVs_3
   CustomizedUVs_4
   CustomizedUVs_5
   CustomizedUVs_6
   CustomizedUVs_7
   PixelDepthOffset
   ShadingModel
   Displacement
  )
  n53 = Constant() {r=0.01}
  n54 = ScalarParameter() {defaultValue=0.6, parameterName=Spec-LowQ}
  n55 = ScalarParameter() {defaultValue=0.6, parameterName=Rough-LowQ}
  n56 = Multiply(A=@n9.RGB, B=@n21)
  n57 = Multiply(A=Add(A=Add(A=@n7.Vector Converted to Angle, B=@n36), B=@n41), B=@n23.RGB)
  n58 = Add(A=Add(A=LinearInterpolate(A=@n56, B=@n57, Alpha=@n40.Result), B=@n8), B=@n50)
  n59 = Constant3Vector() {constant=<rgba 0 0 1 0>}
  MaterialFunctionCall(In=QualitySwitch(Default=@n52, Low=MakeMaterialAttributes(BaseColor=@n53, Specular=@n54, Roughness=@n55, EmissiveColor=@n58, Normal=@n59), High=@n52)) {materialFunction=&MF_TechArtOverrides_accessories}
  n60 = Constant3Vector() {desc="B=Size", constant=<rgba 0.4 0.55 0.55 1>}
  n61 = Distance(A=@n12.UVs, B=ComponentMask(None=@n60) {g=true, r=true})
  n62 = ComponentMask(None=@n60) {b=true}
  n63 = Constant()
  n64 = Constant() {r=1}
  n65 = Constant() {r=0.04}
  Multiply(A=OneMinus(None=If(A=@n61, B=@n62, "A > B"=@n63, "A < B"=@n64)), B=If(A=@n61, B=Add(A=@n62, B=@n65), "A > B"=@n63, "A < B"=@n64))
 ]
}
